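; Lobby lens flare actor that keeps the NS_Asteria_SunFlare_Lobby effect in line with the sun as seen from the camera.
; BeginPlay activates the effect; its camera validity and spectator checks lead to empty branches.
; Each tick reads the editor viewport or player camera position, then places the effect FlareDistance away toward SunPosition.
; SunPosition is scaled by 1e10 so it behaves as a near-infinitely distant point.
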(event EventBeginPlay
  (bind _returnvalue (Game|GetPlayerCameraManager 0))
  (Utilities|IsValid _returnvalue
    (:Is%20Valid
      (if (and (Utilities|String|Contains (Utilities|GetDisplayName _returnvalue) "Spectator") (Editor|InEditororPIE))
        _
        (else
          _)))
    (:Is%20Not%20Valid
      _))
  (Components|Activation|Activate (Variables|Default|GetNS_Asteria_SunFlare_Lobby) false))

(event Collision|EventActorBeginOverlap
  (params (OtherActor :type Actor%20Object%20Reference)))

(event EventTick
  (params (DeltaSeconds :type Float%20%28single-precision%29))
  (bind _returnvalue (Game|GetPlayerCameraManager 0))
  (if (Deprecated|UEFN|IsinEditor)
    (Variables|Default|SetCameraPosition (Editor|GetLevelViewportCameraInfo))
    (else
      (Utilities|IsValid _returnvalue
        (:Is%20Valid
          (Variables|Default|SetCameraPosition (Camera|GetCameraLocation _returnvalue))))))
  (Transformation|SetWorldLocationAndRotation (Variables|Default|GetNS_Asteria_SunFlare_Lobby) (+ (* (Math|Vector|Normalize (- (* (Variables|Default|GetSunPosition) 1e+10) (Variables|Default|GetCameraPosition))) (Variables|Default|GetFlareDistance)) (Variables|Default|GetCameraPosition)) "0, 0, 0" false false))

(fn ConstructionScript)
