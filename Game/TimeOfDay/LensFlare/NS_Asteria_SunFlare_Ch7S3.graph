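# Niagara sun lens flare for the time-of-day sky, built from GPU sprite emitters.
# Flare_DiscsRow and ChromaHoop project onto the camera plane, placing a row of colored discs and a hoop that react to distance from screen center.
# DirtyLnes001 draws a large screen-edge-aware sprite at the owner; LargeGlow001 is disabled.
# Scratch modules supply occlusion-based visibility to fade each element, then atmosphere color and User.GlobalOpacity tint and scale it.
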NiagaraSystem /Game/TimeOfDay/LensFlare/NS_Asteria_SunFlare_Ch7S3
dependencies {
 usedDataInterfaces [
  &NiagaraDataInterfaceCamera
  /Script/Niagara.NiagaraDataInterfaceOcclusion
  &NiagaraDataInterfaceCurve
  &NiagaraDataInterfaceColorCurve
 ]
 usedDynamicInputs []
 usedModules [&SystemState]
}
emitters {
 ChromaHoop {
  simTarget=GPUComputeSim
  emitterUpdateScript [
   &EmitterState()
   &SpawnBurst_Instantaneous(Use Spawn Probability=false, UseLoopCountLimit=true)
  ]
  particleSpawnScript [&InitializeParticle(
    Color=<rgba 1 1 1 0.1>
    Lifetime=5
    Sprite Size Mode=::NewEnumerator3
    Uniform Sprite Size=100
    UsePositionOffset=true
    Write Material Random=true
   )]
  particleUpdateScript [
   /Game/Effects/Niagara/Modules/LensFlares/FN_ProjectToCameraPlane()
   /Game/Effects/Niagara/Modules/LensFlares/FN_CameraScreenCoords()
   &self:ChromaHoop_2.NiagaraScratchPadContainer_1.Occlusion_0(Light Source Radius=5, Samples per ring=1) {enabled=false, module=Occlusion}
   &self:ScratchModule(NumberOfRings=3, OcclusionSize=4000, SamplesPerRing=8) {module=ScratchModule}
   &ParticleState(Kill Particles When Lifetime Has Elapsed=false, Let Infinitely Lived Particles Die When Emitter Deactivates=true)
   &ScaleSpriteSize(
    Uniform Curve Index=&ReturnNormalizedExecIndex()
    Uniform Curve Scale=&RandomRangeFloat(Minimum=0.2, Maximum=1.6, Random Seed=5)
    Uniform Curve Sprite Scale=DataInterface() {propertyValues {Curve {keys [
        {arriveTangent=-2.927, interpMode=RCIM_Cubic, leaveTangent=-2.927, tangentMode=RCTM_User, time=0, value=1}
        {arriveTangent=-0.15, interpMode=RCIM_Cubic, leaveTangent=-0.15, time=0.5, value=0.3}
        {arriveTangent=2.1125, interpMode=RCIM_Cubic, leaveTangent=2.1125, tangentMode=RCTM_User, time=1, value=0.85}
       ]}}}
   ) {enabled=false}
   /SetVariables(
    Particles.RadialDistribution:NiagaraFloat=&FloatFromCurve(
     FloatCurve=DataInterface() {propertyValues {Curve {
        postInfinityExtrap=RCCE_Cycle
        preInfinityExtrap=RCCE_Cycle
        keys [
         {
          arriveTangent=2.7062
          arriveTangentWeight=0.3421
          interpMode=RCIM_Cubic
          leaveTangent=2.7062
          leaveTangentWeight=0.3421
          tangentMode=RCTM_User
          time=0
          value=0
         }
         {arriveTangent=0.6379, interpMode=RCIM_Cubic, leaveTangent=0.6379, time=0.1609, value=0.1942}
         {arriveTangent=0.2894, interpMode=RCIM_Cubic, leaveTangent=0.2894, tangentMode=RCTM_User, time=0.3696, value=0.2358}
         {arriveTangent=1.537, arriveTangentWeight=0.1667, interpMode=RCIM_Cubic, leaveTangent=1.537, leaveTangentWeight=0.1667, time=0.5, value=0.5}
         {arriveTangent=0.674, interpMode=RCIM_Cubic, leaveTangent=0.674, time=0.6358, value=0.6449}
         {arriveTangent=0.3327, interpMode=RCIM_Cubic, leaveTangent=0.3327, tangentMode=RCTM_User, time=0.8641, value=0.7454}
         {
          arriveTangent=4.0752
          arriveTangentWeight=0.4541
          interpMode=RCIM_Cubic
          leaveTangent=4.0752
          leaveTangentWeight=0.4541
          tangentMode=RCTM_User
          time=1
          value=1
         }
        ]
       }}}
     CurveIndex=&ReturnNormalizedExecIndex()
    )
    Particles.RadialWeight:NiagaraFloat=&FloatFromCurve(
     FloatCurve=DataInterface() {propertyValues {Curve {keys [
         {arriveTangent=-1.4689, interpMode=RCIM_Cubic, leaveTangent=-1.4689, tangentMode=RCTM_User, time=0, value=1}
         {interpMode=RCIM_Cubic, tangentMode=RCTM_User, time=0.5, value=0.66}
         {arriveTangent=1.5059, interpMode=RCIM_Cubic, leaveTangent=1.5059, tangentMode=RCTM_User, time=1, value=1}
        ]}}}
     CurveIndex=&RemapRange(Input Value To Remap=&ReturnNormalizedExecIndex(), Input Min=0, Input Max=1, Output Min=-1, Output Max=1)
    )
   ) {isSetParameters=true, module=SetVariables_#1}
   /SetVariables(Particles.Position=&LerpPosition(
     A=$Particles.PlanePosition
     B=$Particles.ProjectedPosition
     Alpha=&RemapRange(Input Value To Remap=$Particles.RadialDistribution, Input Min=0, Input Max=1, Output Min=0.5, Output Max=1)
    )) {isSetParameters=true, module=SetVariables_#2}
   &Color(
    Color=&ColorFromCurve(
     ColorCurve=DataInterface() {propertyValues {
       AlphaCurve {keys [<at 0 1 RCIM_Constant>,<at 1 1 RCIM_Constant>]}
       BlueCurve {keys [<at 0.0674 0 RCIM_Constant>,<at 0.1132 0.0019 RCIM_Constant>,<at 0.2071 0.3854 RCIM_Constant>,<at 0.3323 0.0981 RCIM_Constant>,<at 0.4317 0 RCIM_Constant>,<at 0.5992 0.3854 RCIM_Constant>,<at 0.6669 0.3854 RCIM_Constant>,<at 0.7506 0.8385 RCIM_Constant>,<at 0.8066 0 RCIM_Constant>,<at 0.927 0 RCIM_Constant>,<at 0.9751 0.4479 RCIM_Constant>]}
       GreenCurve {keys [<at 0.0674 0.0521 RCIM_Constant>,<at 0.1132 0.0403 RCIM_Constant>,<at 0.2071 0.1767 RCIM_Constant>,<at 0.3323 0 RCIM_Constant>,<at 0.4317 0.5781 RCIM_Constant>,<at 0.5992 0.0388 RCIM_Constant>,<at 0.6669 0 RCIM_Constant>,<at 0.7506 0.1147 RCIM_Constant>,<at 0.8066 0.3125 RCIM_Constant>,<at 0.927 0.6458 RCIM_Constant>,<at 0.9751 0.0596 RCIM_Constant>]}
       RedCurve {keys [<at 0.0674 0.026 RCIM_Constant>,<at 0.1132 0.3854 RCIM_Constant>,<at 0.2071 0.0345 RCIM_Constant>,<at 0.3323 1 RCIM_Constant>,<at 0.4317 0.1432 RCIM_Constant>,<at 0.5992 0.2895 RCIM_Constant>,<at 0.6669 0.1222 RCIM_Constant>,<at 0.7506 0.3762 RCIM_Constant>,<at 0.8066 0.1236 RCIM_Constant>,<at 0.927 0.5112 RCIM_Constant>,<at 0.9751 0.1218 RCIM_Constant>]}
      }}
     CurveIndex=&Add_Float(A=&RandomRangeFloat(Minimum=-0.3, Maximum=0.01), B=$Particles.RadialDistribution)
    )
    Scale Alpha=$Particles.Visibility
   ) {enabled=false}
   /SetVariables(Particles.CenterFalloff:NiagaraFloat=&FloatFromCurve(
     FloatCurve=DataInterface() {propertyValues {Curve {keys [
         {tangentMode=RCTM_User, time=0, value=1}
         {arriveTangent=-0.1986, interpMode=RCIM_Cubic, leaveTangent=-0.1986, tangentMode=RCTM_User, time=0.5, value=0.95}
         {arriveTangent=-2.4201, interpMode=RCIM_Cubic, leaveTangent=-2.4201, tangentMode=RCTM_User, time=1.3, value=0.1}
        ]}}}
     CurveIndex=$Particles.DistanceFromCenter
    )) {isSetParameters=true, module=SetVariables_#3}
   &ScaleColor(
    Color Value To Scale=$Particles.Color
    Scale Mode=::NewEnumerator1
    Scale RGBA=&MakeVector4_FromFloat(Float=$Particles.CenterFalloff)
    ScaleRGBA=true
   ) {enabled=false}
   /SetVariables(Particles.SpriteAlignment=&Subtract_Position(A=$Particles.PlanePosition, B=$Particles.ProjectedPosition)) {isSetParameters=true, module=SetVariables_#4}
   /SetVariables(Particles.SpriteFacing=&Multiply_VectorByFloat(
     Vector=&Lerp_Vector(
      A=&NormalizeVector(
       Vector To Normalize=&Subtract_Position(A=$Particles.ProjectedPosition, B=$Particles.PlanePosition)
       Safe Normalize=true
       Safe Fallback Threshold=0
       Safe Fallback Vector=<0 0 0.001>
      )
      B=$Particles.PlaneNormal
      Alpha=&RemapRange(Input Value To Remap=$Particles.DistanceFromCenter, Input Min=1, Input Max=0, Output Min=0.3, Output Max=1)
     )
     Float=1
    )) {isSetParameters=true, module=SetVariables_#5}
   &ScaleSpriteSize(
    Uniform Curve Index=$Particles.DistanceFromCenter
    Uniform Curve Sprite Scale=DataInterface() {propertyValues {Curve {keys [{tangentMode=RCTM_User, time=0, value=0.25}, {tangentMode=RCTM_User, time=1, value=1}]}}}
   ) {module=ScaleSpriteSize001}
   &ScaleColor(
    Scale RGB=&VectorFromFloat(Value=&FloatFromCurve(
      FloatCurve=DataInterface() {propertyValues {Curve {keys [
          {arriveTangent=4.1785, interpMode=RCIM_Cubic, leaveTangent=4.1785, tangentMode=RCTM_User, time=0.1, value=0}
          {arriveTangent=1.1179, interpMode=RCIM_Cubic, leaveTangent=1.1179, tangentMode=RCTM_User, time=0.3, value=0.6}
          {arriveTangent=0.0224, interpMode=RCIM_Cubic, leaveTangent=0.0224, tangentMode=RCTM_User, time=0.8, value=1}
          {arriveTangent=-4.295, interpMode=RCIM_Cubic, leaveTangent=-4.295, tangentMode=RCTM_User, time=1, value=0}
         ]}}}
      CurveIndex=$Particles.DistanceFromCenter
      Scale Curve=&Multiply_Float(A=1, B=$Particles.Visibility)
     ))
    ScaleA=true
    ScaleRGB=true
   ) {module=ScaleColor001}
   /SetVariables(Particles.SpriteSize=&MakeVector2D(
     X=&Multiply_Float(
      A=&MakeFloatFromVector2d(Vector 2D=$Particles.SpriteSize)
      B=&FloatFromCurve(
       FloatCurve=DataInterface() {propertyValues {Curve {keys [
           {arriveTangent=-0.2123, interpMode=RCIM_Cubic, leaveTangent=-0.2123, tangentMode=RCTM_User, time=0, value=1.5}
           {interpMode=RCIM_Cubic, tangentMode=RCTM_User, time=1, value=1}
          ]}}}
       CurveIndex=$Particles.CenterFalloff
      )
     )
     Y=&MakeFloatFromVector2d(Vector 2D=$Particles.SpriteSize, Channel=::NewEnumerator1)
    )) {enabled=false, isSetParameters=true, module=SetVariables_#6}
   &DynamicMaterialParameters(
    Index 3 Param 4=$Particles.CenterFalloff
    Module.Param0WriteEnabled3=false
    Module.Param1WriteEnabled3=false
    Module.Param2WriteEnabled3=false
    Module.Param3WriteEnabled3=true
    Write Parameter Index 0=false
    Write Parameter Index 3=true
   )
   &ScaleColor(
    Color Value To Scale=$Particles.Color
    Scale Alpha=&Multiply_Float(A=$User.GlobalOpacity, B=1)
    Scale RGB=&Multiply_Vector(A=&MakeVectorFromLinearColorRGB(Color=$Particles.AtmosphereColor), B=<1 1 1>)
    ScaleA=true
    ScaleRGB=true
   ) {module=ScaleColor002}
  ]
  rendererClasses [&NiagaraSpriteRendererProperties]
  renderers [{rendererClass=&NiagaraSpriteRendererProperties}]
 }
 DirtyLnes001 {
  simTarget=GPUComputeSim
  emitterUpdateScript [
   &EmitterState()
   &SpawnBurst_Instantaneous(Use Spawn Probability=false, UseLoopCountLimit=true)
  ]
  particleSpawnScript [&InitializeParticle(
    Color=<rgba 2.5 2.5 2.5 1>
    Sprite Rotation Mode=::NewEnumerator2
    Sprite Size Mode=::NewEnumerator3
    Uniform Sprite Size=&Multiply_Float(A=2250000, B=$User.GlobalScale)
    UsePositionOffset=false
    Write Material Random=true
   )]
  particleUpdateScript [
   &ParticleState(Kill Particles When Lifetime Has Elapsed=false, Let Infinitely Lived Particles Die When Emitter Deactivates=true)
   &self:MaskInScreenSpace() {module=MaskInScreenSpace}
   &self:ScratchModule(NumberOfRings=3, OcclusionSize=4000, SamplesPerRing=8) {module=ScratchModule}
   &ScaleColor(Scale RGB=&Multiply_VectorByFloat(Vector=<1 1 1>, Float=$Particles.Visibility), ScaleA=true, ScaleRGB=true) {module=ScaleColor001}
   &ScaleSpriteSize(
    Uniform Curve Index=$Transient.DistanceFromScreenEdge
    Uniform Curve Scale=1.25
    Uniform Curve Sprite Scale=DataInterface() {propertyValues {Curve {keys [{tangentMode=RCTM_User, time=0, value=0.5}, {tangentMode=RCTM_User, time=1, value=1}]}}}
   )
   /SetVariables(Particles.Position=$Engine.Owner.Position) {isSetParameters=true}
   &ScaleColor(
    Color Value To Scale=$Particles.Color
    Scale Alpha=&Multiply_Float(A=$User.GlobalOpacity, B=1)
    Scale RGB=&Multiply_Vector(A=&MakeVectorFromLinearColorRGB(Color=$Particles.AtmosphereColor), B=<1 1 1>)
    ScaleA=true
    ScaleRGB=true
   )
  ]
  rendererClasses [&NiagaraSpriteRendererProperties]
  renderers [
   {rendererClass=&NiagaraSpriteRendererProperties}
   {rendererClass=&NiagaraSpriteRendererProperties, rendererIndex=1}
  ]
 }
 Flare_DiscsRow {
  simTarget=GPUComputeSim
  emitterUpdateScript [
   &EmitterState()
   &SpawnBurst_Instantaneous(Spawn Count=18, Use Spawn Probability=false, UseLoopCountLimit=true)
  ]
  particleSpawnScript [&InitializeParticle(
    Color=<rgba 0 1 1 0.1>
    Lifetime=5
    Sprite Size Mode=::NewEnumerator1
    Uniform Sprite Size Max=40
    UsePositionOffset=true
    Write Material Random=true
   )]
  particleUpdateScript [
   &FN_ProjectToCameraPlane()
   &FN_CameraScreenCoords()
   &self:Flare_DiscsRow.NiagaraScratchPadContainer_1.Occlusion_0(Light Source Radius=5, Samples per ring=1) {enabled=false, module=Occlusion}
   &self:ScratchModule(NumberOfRings=3, OcclusionSize=4000, SamplesPerRing=8) {module=ScratchModule}
   &ParticleState(Kill Particles When Lifetime Has Elapsed=false, Let Infinitely Lived Particles Die When Emitter Deactivates=true)
   &ScaleSpriteSize(
    Uniform Curve Index=&ReturnNormalizedExecIndex()
    Uniform Curve Scale=&RandomRangeFloat(Minimum=0.2, Maximum=1.6, Random Seed=5)
    Uniform Curve Sprite Scale=DataInterface() {propertyValues {Curve {keys [
        {arriveTangent=-2.927, interpMode=RCIM_Cubic, leaveTangent=-2.927, tangentMode=RCTM_User, time=0, value=1}
        {arriveTangent=-0.15, interpMode=RCIM_Cubic, leaveTangent=-0.15, time=0.5, value=0.3}
        {arriveTangent=2.1125, interpMode=RCIM_Cubic, leaveTangent=2.1125, tangentMode=RCTM_User, time=1, value=0.85}
       ]}}}
   )
   /SetVariables(
    Particles.RadialDistribution:NiagaraFloat=&FloatFromCurve(
     FloatCurve=DataInterface() {propertyValues {Curve {
        postInfinityExtrap=RCCE_Cycle
        preInfinityExtrap=RCCE_Cycle
        keys [
         {
          arriveTangent=2.7062
          arriveTangentWeight=0.3421
          interpMode=RCIM_Cubic
          leaveTangent=2.7062
          leaveTangentWeight=0.3421
          tangentMode=RCTM_User
          time=0
          value=0
         }
         {arriveTangent=0.6379, interpMode=RCIM_Cubic, leaveTangent=0.6379, time=0.1609, value=0.1942}
         {arriveTangent=0.2894, interpMode=RCIM_Cubic, leaveTangent=0.2894, tangentMode=RCTM_User, time=0.3696, value=0.2358}
         {arriveTangent=1.537, arriveTangentWeight=0.1667, interpMode=RCIM_Cubic, leaveTangent=1.537, leaveTangentWeight=0.1667, time=0.5, value=0.5}
         {arriveTangent=0.674, interpMode=RCIM_Cubic, leaveTangent=0.674, time=0.6358, value=0.6449}
         {arriveTangent=0.3327, interpMode=RCIM_Cubic, leaveTangent=0.3327, tangentMode=RCTM_User, time=0.8641, value=0.7454}
         {
          arriveTangent=4.0752
          arriveTangentWeight=0.4541
          interpMode=RCIM_Cubic
          leaveTangent=4.0752
          leaveTangentWeight=0.4541
          tangentMode=RCTM_User
          time=1
          value=1
         }
        ]
       }}}
     CurveIndex=&ReturnNormalizedExecIndex()
    )
    Particles.RadialWeight:NiagaraFloat=&FloatFromCurve(
     FloatCurve=DataInterface() {propertyValues {Curve {keys [
         {arriveTangent=-1.4689, interpMode=RCIM_Cubic, leaveTangent=-1.4689, tangentMode=RCTM_User, time=0, value=1}
         {interpMode=RCIM_Cubic, tangentMode=RCTM_User, time=0.5, value=0.66}
         {arriveTangent=1.5059, interpMode=RCIM_Cubic, leaveTangent=1.5059, tangentMode=RCTM_User, time=1, value=1}
        ]}}}
     CurveIndex=&RemapRange(Input Value To Remap=&ReturnNormalizedExecIndex(), Input Min=0, Input Max=1, Output Min=-1, Output Max=1)
    )
   ) {isSetParameters=true, module=SetVariables_#1}
   /SetVariables(Particles.Position=&LerpPosition(
     A=$Particles.PlanePosition
     B=$Particles.ProjectedPosition
     Alpha=&RemapRange(Input Value To Remap=$Particles.RadialDistribution, Input Min=0, Input Max=1, Output Min=-1, Output Max=1)
    )) {isSetParameters=true, module=SetVariables_#2}
   &Color(Color=&ColorFromCurve(
     ColorCurve=DataInterface() {propertyValues {
       AlphaCurve {keys [<at 0 1 RCIM_Constant>,<at 1 1 RCIM_Constant>]}
       BlueCurve {keys [<at 0.0674 0 RCIM_Constant>,<at 0.1132 0.0019 RCIM_Constant>,<at 0.2071 0.3854 RCIM_Constant>,<at 0.3323 0.0981 RCIM_Constant>,<at 0.4317 0 RCIM_Constant>,<at 0.5992 0.3854 RCIM_Constant>,<at 0.6669 0.3854 RCIM_Constant>,<at 0.7506 0.8385 RCIM_Constant>,<at 0.8066 0 RCIM_Constant>,<at 0.927 0 RCIM_Constant>,<at 0.9751 0.4479 RCIM_Constant>]}
       GreenCurve {keys [<at 0.0674 0.0521 RCIM_Constant>,<at 0.1132 0.0403 RCIM_Constant>,<at 0.2071 0.1767 RCIM_Constant>,<at 0.3323 0 RCIM_Constant>,<at 0.4317 0.5781 RCIM_Constant>,<at 0.5992 0.0388 RCIM_Constant>,<at 0.6669 0 RCIM_Constant>,<at 0.7506 0.1147 RCIM_Constant>,<at 0.8066 0.3125 RCIM_Constant>,<at 0.927 0.6458 RCIM_Constant>,<at 0.9751 0.0596 RCIM_Constant>]}
       RedCurve {keys [<at 0.0674 0.026 RCIM_Constant>,<at 0.1132 0.3854 RCIM_Constant>,<at 0.2071 0.0345 RCIM_Constant>,<at 0.3323 1 RCIM_Constant>,<at 0.4317 0.1432 RCIM_Constant>,<at 0.5992 0.2895 RCIM_Constant>,<at 0.6669 0.1222 RCIM_Constant>,<at 0.7506 0.3762 RCIM_Constant>,<at 0.8066 0.1236 RCIM_Constant>,<at 0.927 0.5112 RCIM_Constant>,<at 0.9751 0.1218 RCIM_Constant>]}
      }}
     CurveIndex=&Add_Float(A=&RandomRangeFloat(Minimum=-0.3, Maximum=0.01), B=$Particles.RadialDistribution)
    ))
   /SetVariables(Particles.CenterFalloff:NiagaraFloat=&FloatFromCurve(
     FloatCurve=DataInterface() {propertyValues {Curve {keys [
         {tangentMode=RCTM_User, time=0, value=1}
         {arriveTangent=-0.1986, interpMode=RCIM_Cubic, leaveTangent=-0.1986, tangentMode=RCTM_User, time=0.5, value=0.95}
         {arriveTangent=-2.4201, interpMode=RCIM_Cubic, leaveTangent=-2.4201, tangentMode=RCTM_User, time=1.3, value=0.1}
        ]}}}
     CurveIndex=$Particles.DistanceFromCenter
    )) {isSetParameters=true, module=SetVariables_#3}
   &ScaleColor(
    Color Value To Scale=$Particles.Color
    Scale Mode=::NewEnumerator1
    Scale RGBA=&MakeVector4_FromFloat(Float=&Multiply_Float(A=$Particles.CenterFalloff, B=$Particles.Visibility))
    ScaleRGBA=true
   )
   /SetVariables(Particles.SpriteAlignment=&Subtract_Position(A=$Particles.PlanePosition, B=$Particles.ProjectedPosition)) {enabled=false, isSetParameters=true, module=SetVariables_#4}
   /SetVariables(Particles.SpriteFacing=&Multiply_VectorByFloat(
     Vector=&Lerp_Vector(
      A=&NormalizeVector(
       Vector To Normalize=&Subtract_Position(A=$Particles.ProjectedPosition, B=$Particles.PlanePosition)
       Safe Normalize=true
       Safe Fallback Threshold=0
       Safe Fallback Vector=<0 0 0.001>
      )
      B=$Particles.PlaneNormal
      Alpha=&RemapRange(Input Value To Remap=$Particles.DistanceFromCenter, Input Min=1, Input Max=0, Output Min=0.3, Output Max=1)
     )
     Float=1
    )) {isSetParameters=true, module=SetVariables_#5}
   /SetVariables(Particles.SpriteSize=&MakeVector2D(
     X=&Multiply_Float(
      A=&MakeFloatFromVector2d(Vector 2D=$Particles.SpriteSize)
      B=&FloatFromCurve(
       FloatCurve=DataInterface() {propertyValues {Curve {keys [
           {arriveTangent=-0.2123, interpMode=RCIM_Cubic, leaveTangent=-0.2123, tangentMode=RCTM_User, time=0, value=1.5}
           {interpMode=RCIM_Cubic, tangentMode=RCTM_User, time=1, value=1}
          ]}}}
       CurveIndex=$Particles.CenterFalloff
      )
     )
     Y=&MakeFloatFromVector2d(Vector 2D=$Particles.SpriteSize, Channel=::NewEnumerator1)
    )) {isSetParameters=true, module=SetVariables_#6}
   &DynamicMaterialParameters(
    Index 3 Param 4=$Particles.CenterFalloff
    Module.Param0WriteEnabled3=false
    Module.Param1WriteEnabled3=false
    Module.Param2WriteEnabled3=false
    Module.Param3WriteEnabled3=true
    Write Parameter Index 0=false
    Write Parameter Index 3=true
   )
   &ScaleColor(
    Color Value To Scale=$Particles.Color
    Scale Alpha=&Multiply_Float(A=$User.GlobalOpacity, B=1)
    Scale RGB=&Multiply_Vector(A=&MakeVectorFromLinearColorRGB(Color=$Particles.AtmosphereColor), B=<1 1 1>)
    ScaleA=true
    ScaleRGB=true
   ) {module=ScaleColor001}
  ]
  rendererClasses [&NiagaraSpriteRendererProperties]
  renderers [{rendererClass=&NiagaraSpriteRendererProperties}]
 }
 LargeGlow001 {
  bEnabled=false
  simTarget=GPUComputeSim
  emitterUpdateScript [
   &EmitterState()
   &SpawnBurst_Instantaneous(Use Spawn Probability=false, UseLoopCountLimit=true)
  ]
  particleSpawnScript [&InitializeParticle(
    Color=<rgba 0.25 0.1536 0.049 0.75>
    Sprite Rotation Mode=::NewEnumerator2
    Sprite Size Mode=::NewEnumerator3
    Uniform Sprite Size=&Multiply_Float(A=2400000, B=$User.GlobalScale)
    UsePositionOffset=false
    Write Material Random=true
   )]
  particleUpdateScript [
   &ParticleState(Kill Particles When Lifetime Has Elapsed=false, Let Infinitely Lived Particles Die When Emitter Deactivates=true)
   &self:MaskInScreenSpace() {module=MaskInScreenSpace}
   &self:ScratchModule(NumberOfRings=3, OcclusionSize=4000, SamplesPerRing=8) {module=ScratchModule}
   &ScaleColor(
    Curve Index=$Transient.DistanceFromScreenEdge
    Linear Color Curve=DataInterface() {propertyValues {
      AlphaCurve {keys [<at 0 1>]}
      BlueCurve {keys [<at 0 0.6825>,<at 0.8973 0.2113>]}
      GreenCurve {keys [<at 0 0.9417>,<at 0.8973 0.6541>]}
      RedCurve {keys [<at 0 1>,<at 0.8973 1>]}
     }}
    Scale Mode=::NewEnumerator2
   ) {module=ScaleColor002}
   /SetVariables(Particles.Position=$Engine.Owner.Position) {isSetParameters=true}
   &ScaleColor(Scale Alpha=$Particles.Visibility, ScaleA=true, ScaleRGB=true)
   &ScaleColor(
    Color Value To Scale=$Particles.Color
    Scale Alpha=&Multiply_Float(A=$User.GlobalOpacity, B=1)
    Scale RGB=&Multiply_Vector(A=&MakeVectorFromLinearColorRGB(Color=$Particles.AtmosphereColor), B=<1 1 1>)
    ScaleA=true
    ScaleRGB=true
   ) {module=ScaleColor001}
  ]
  rendererClasses [&NiagaraSpriteRendererProperties]
  renderers [{rendererClass=&NiagaraSpriteRendererProperties}]
 }
}
summary {
 systemName=NS_Asteria_SunFlare_Ch7S3
 emitters [
  {emitterName=Flare_DiscsRow, simTarget=GPUComputeSim, rendererClasses [&NiagaraSpriteRendererProperties]}
  {emitterName=ChromaHoop, simTarget=GPUComputeSim, rendererClasses [&NiagaraSpriteRendererProperties]}
  {emitterName=DirtyLnes001, simTarget=GPUComputeSim, rendererClasses [&NiagaraSpriteRendererProperties]}
  {bEnabled=false, emitterName=LargeGlow001, simTarget=GPUComputeSim, rendererClasses [&NiagaraSpriteRendererProperties]}
 ]
 userVariables [
  User.GlobalOpacity {type=NiagaraFloat, defaultValue=1, description=""}
  User.GlobalScale {type=NiagaraFloat, defaultValue=0.75, description=""}
 ]
}
